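# MF_Translucency is a library-exposed material function in the Megascans foliage material that computes a translucency color from albedo and a grayscale input.
# It multiplies the grayscale input by Translucency Strength, then tints that by the albedo.
# The result is desaturated by one minus Translucency Desaturation, so higher parameter values keep more color and the default desaturates only slightly.
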MaterialFunction /Bridge/MSPresets/M_MS_Foliage_Material/Functions/MF_Translucency
graph {
 properties {bExposeToLibrary=true}
 nodes [
  n1 = FunctionInput() {inputName=Albedo}
  n2 = Multiply(
   A=FunctionInput() {inputName=Grayscale, inputType=FunctionInput_Scalar, sortPriority=1}
   B=ScalarParameter() {defaultValue=1, group=06 - Translucency, parameterName=Translucency Strength, sortPriority=0}
  )
  n3 = ScalarParameter() {defaultValue=0.95, group=06 - Translucency, parameterName=Translucency Desaturation, sortPriority=1}
  FunctionOutput(None=Desaturation(None=Multiply(A=@n1, B=@n2), Fraction=OneMinus(None=@n3)))
 ]
}
