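# Masked, two-sided foliage master material for Bridge (Megascans) plants, with parameters for textures, color, roughness, opacity, normal strength and wind.
# Albedo gets an overlay tint and SpeedTree color variation, then enters MF_Translucency with ORT blue.
# ORT red and green supply scaled opacity mask and roughness.
# An optional static switch enables SimpleGrassWind, which vertex color green fades toward zero as it rises.
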Material /Bridge/MSPresets/M_MS_Foliage_Material/M_MS_Foliage_Material
graph {
 properties {
  blendMode=BLEND_Masked
  pixelDepthOffsetMode=PDOM_Legacy
  shadingModel=MSM_TwoSidedFoliage
  twoSided=true
  usedShadingModels=Two Sided Foliage
 }
 nodes [
  n1 = TextureSampleParameter2D() {
   group=01- Input Textures
   parameterName=Normal
   samplerSource=SSM_Wrap_WorldGroupSettings
   samplerType=SAMPLERTYPE_Normal
   sortPriority=2
   texture=/Bridge/MSPresets/MSTextures/Placeholder_Normal
  }
  MaterialFunctionCall(Normal=@n1.RGB, Flatness=OneMinus(None=ScalarParameter() {defaultValue=1, group=05 - Normal, parameterName=Normal Intensity})) {materialFunction=&FlattenNormal}
  n2 = TextureSampleParameter2D() {
   group=01- Input Textures
   parameterName=Albedo
   samplerSource=SSM_Wrap_WorldGroupSettings
   sortPriority=0
   texture=/Engine/EngineMaterials/DefaultDiffuse
  }
  n3 = MaterialFunctionCall(Base=@n2.RGB, Blend=VectorParameter() {group=02 - Albedo, parameterName=Color Overlay, defaultValue=<rgba 0.5 0.5 0.5 1>}.RGB) {materialFunction=&Blend_Overlay}
  n4 = ScalarParameter() {defaultValue=1, group=02 - Albedo, parameterName=Overlay Intensity, sliderMax=1.5, sortPriority=1}
  n5 = TextureSampleParameter2D() {
   group=01- Input Textures
   parameterName=ORT
   samplerType=SAMPLERTYPE_LinearColor
   sortPriority=3
   texture=/Bridge/MSPresets/MSTextures/WhitePlaceholder
  }
  MaterialFunctionCall!(
   Albedo=MaterialFunctionCall!(
    BaseColor=LinearInterpolate(A=@n2.RGB, B=@n3.Result, Alpha=@n4)
    Amount=ScalarParameter() {group=02 - Albedo, parameterName=Color Variation, sortPriority=2}
   ) {desc=ms_BaseColor, materialFunction=/Engine/Functions/Engine_MaterialFunctions01/SpeedTree/SpeedTreeColorVariation}.Result
   Grayscale=@n5.B
  ) {materialFunction=/Bridge/MSPresets/M_MS_Foliage_Material/Functions/MF_Translucency}
  Multiply(A=@n5.G, B=ScalarParameter() {defaultValue=1, group=03- Roughness, parameterName=Roughness Intensity}) {desc=ms_Roughness}
  Multiply(A=@n5.R, B=ScalarParameter() {defaultValue=1, group=04 - Opacity, parameterName=Opacity Intensity}) {desc=ms_OpacityMask}
  n6 = ScalarParameter() {defaultValue=0.15, group=07 - Wind, parameterName=Wind Intensity, sortPriority=1}
  n7 = ScalarParameter() {defaultValue=0.5, group=07 - Wind, parameterName=Wind Height, sortPriority=2}
  n8 = ScalarParameter() {defaultValue=0.5, group=07 - Wind, parameterName=Wind Speed, sortPriority=3}
  n9 = ScalarParameter() {group=07 - Wind, parameterName=Additional WPO, sortPriority=4}
  n10 = Constant()
  StaticSwitchParameter(
   True=LinearInterpolate(
    A=MaterialFunctionCall!(WindIntensity=@n6, WindWeight=@n7, WindSpeed=@n8, AdditionalWPO=@n9) {materialFunction=/Engine/Functions/Engine_MaterialFunctions01/WorldPositionOffset/SimpleGrassWind}.Result
    B=@n10
    Alpha=VertexColor().G
   )
   False=@n10
  ) {group=07 - Wind, parameterName=EnableGrassWind}
 ]
}
